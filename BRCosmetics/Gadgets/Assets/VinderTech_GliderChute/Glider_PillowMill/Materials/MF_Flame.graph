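# Material function for the Pillow Mill glider that generates a stylized flame mask, described as similar to Chainmail.
# A panned, scaled trail texture is multiplied by a soft sphere mask, bypassable for radial meshes, then blends into the plain mask toward high V.
# Dynamic parameters drive vertical speed and a shared UV offset.
# The result is attenuated by camera depth fade and a vertical view-angle term; an MF_Erosion call does not feed the output.
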MaterialFunction /BRCosmetics/Gadgets/Assets/VinderTech_GliderChute/Glider_PillowMill/Materials/MF_Flame
graph {
 properties {description=Generates a Stylized Flame similar to Chainmail.}
 nodes [
  n1 = FunctionInput(Preview=TextureCoordinate()) {bUsePreviewValueAsDefault=true, inputName=UV_Mask, inputType=FunctionInput_Vector2}
  n2 = AppendVector(A=Constant() {r=1}, B=ComponentMask(None=TextureCoordinate()) {g=true})
  n3 = SphereMask(
   A=MaterialFunctionCall(UVs=Multiply(A=@n1, B=@n2), Texture Scale=Constant2Vector() {g=1, r=0.6}) {materialFunction=&ScaleUVsByCenter}.UVs
   B=Constant2Vector() {g=0.5, r=0.5}
  ) {attenuationRadius=0.5, hardnessPercent=0}
  n4 = Multiply(A=@n3) {constB=6}
  n5 = Constant() {r=1}
  n6 = ScalarParameter() {desc=Allow usage with Radial Meshes, parameterName=SkipFlameMask?}
  n7 = TextureCoordinate()
  n8 = VectorParameter() {defaultValue=<rgba 0 0 0 1>}
  n9 = DynamicParameter() {parameterIndex=2, defaultValue=<rgba 1.5 1.7 1 0>, paramNames [Speed,scaleYtex,OffsetX_v,OffsetY_v]}
  n10 = Add(
   A=Add(
    A=FunctionInput(Preview=@n7) {bUsePreviewValueAsDefault=true, inputName=UV_Flame, inputType=FunctionInput_Vector2}
    B=ComponentMask(None=@n8.RGB) {g=true, r=true}
   )
   B=AppendVector(A=@n9.OffsetX_v, B=@n9.OffsetX_v)
  )
  n11 = AppendVector(A=ScalarParameter() {group=Flame, parameterName=Speed X, sliderMax=5, sliderMin=-5}, B=@n9.Speed)
  n12 = ScalarParameter() {defaultValue=1, group=Flame, parameterName=Scale X, sliderMax=5}
  n13 = ScalarParameter() {defaultValue=1.7, group=Flame, parameterName=Scale Y, sliderMax=5}
  n14 = MaterialFunctionCall(XY=MaterialFunctionCall(UVs=Panner(Coordinate=@n10, Speed=@n11) {speedX=1}, Texture Scale=AppendVector(A=@n12, B=@n13)) {materialFunction=&ScaleUVsByCenter}.UVs) {materialFunction=&Swizzle}
  n15 = ComponentMask(None=TextureCoordinate() {uTiling=4}) {g=true}
  n16 = LinearInterpolate(
   A=Multiply(
    A=LinearInterpolate(A=@n4, B=@n5, Alpha=@n6)
    B=TextureSample(UVs=@n14.YX) {texture=/Game/Gadgets/Assets/VinderTech_GliderChute/Glider_DarkStance/FX/T_Glider_DarkStance_Trail}.R
   ) {constA=1}
   B=LinearInterpolate(A=Multiply(A=@n3) {constB=3.5}, B=Constant() {r=1}, Alpha=@n6)
   Alpha=Saturate(None=Power(Base=@n15) {constExponent=6})
  )
  n17 = MaterialFunctionCall(
   Fade Length=ScalarParameter() {defaultValue=20, parameterName=Camera Fade Length, sliderMax=200}
   Fade Offset=ScalarParameter() {defaultValue=20, parameterName=Camera Fade Offset, sliderMax=200}
  ) {materialFunction=&CameraDepthFade}
  n18 = ComponentMask(None=TextureCoordinate()) {r=true}
  n19 = DotProduct(
   A=MaterialFunctionCall() {materialFunction=&CameraDirectionVector}.Result
   B=Constant3Vector() {constant=<rgba 0 0 1 0>}
  )
  n20 = LinearInterpolate(B=Subtract(A=Power(Base=@n18) {constExponent=1}) {constB=0.1}, Alpha=Saturate(None=Power(Base=@n19))) {constA=1}
  FunctionOutput(None=Multiply(A=@n16, B=Multiply(A=@n17.Result, B=@n20) {constA=1}))
  MaterialFunctionCall!(
   Gradient=@n16
   Erosion Amount=ScalarParameter() {defaultValue=0.8, parameterName="Erosion Amount (S)"}
   Softness=ScalarParameter() {defaultValue=0.2, parameterName=softnessErosion}
  ) {materialFunction=/BRCosmetics/Weapons/FORT_Melee/Pickaxe_Turtleneck_Male/FX/Default/MF_Erosion}
  ScalarParameter() {group=Flame, parameterName=OffsetX_v, sliderMax=5}
  ScalarParameter() {group=Flame, parameterName=OffsetY_v, sliderMax=5}
  ScalarParameter() {defaultValue=1, group=Flame, parameterName=Speed Y, sliderMax=5, sliderMin=-5}
 ]
}
