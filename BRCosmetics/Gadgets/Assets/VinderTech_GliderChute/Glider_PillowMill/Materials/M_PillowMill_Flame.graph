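# Masked, unlit, two-sided flame material for Niagara sprite, ribbon and mesh particles, filed with the PillowMill glider cosmetic.
# MF_Flame builds the flame shape from UVs warped by MF_FlameDistortion, colored through a toon curve atlas and a red band, with thresholds from dynamic parameters.
# Branches add sphere-mask erosion, view-angle and depth fades, particle-color tint and time-of-day scaling.
# World-position-offset branches weight bending along the second UV channel.
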Material /BRCosmetics/Gadgets/Assets/VinderTech_GliderChute/Glider_PillowMill/Materials/M_PillowMill_Flame
graph {
 properties {
  bUsedWithNiagaraMeshParticles=true
  bUsedWithNiagaraRibbons=true
  bUsedWithNiagaraSprites=true
  blendMode=BLEND_Masked
  opacityMaskClipValue=0.1
  pixelDepthOffsetMode=PDOM_Legacy
  shadingModel=MSM_Unlit
  twoSided=true
  usedShadingModels=Unlit
 }
 nodes [
  n1 = TextureCoordinate()
  Add(A=@n1, B=AppendVector(A=ScalarParameter() {defaultValue=1, parameterName=Av_X}, B=ScalarParameter() {parameterName=Av_Y}))
  n2 = Add(
   A=Multiply(A=TextureCoordinate(), B=ScalarParameter() {defaultValue=2, parameterName=Vertical Gradient Tiling, sliderMax=3})
   B=NamedRerouteUsage()
  )
  n3 = CurveAtlasRowParameter(CurveTime=Saturate(None=ComponentMask(None=@n2) {g=true})) {
   atlas=&/VFX_CurveAtlas_Toon
   curve=&/CC_EarthLane
   parameterName=Curve_Color
  }
  n4 = DynamicParameter() {defaultValue=<rgba 3.5 1 1 0>, paramNames [redCol,orangeCol,orangeColMin,erode]}
  n5 = SmoothStep(Value=ComponentMask(None=@n1) {g=true})
  n6 = TextureCoordinate()
  n7 = Multiply(A=NamedRerouteUsage(), B=ScalarParameter() {defaultValue=0.1, group=UV Distortion, parameterName=Mask Distortion Strength}) {constB=0}
  n8 = MaterialFunctionCall!(UV_Flame, UV_Mask=Add(A=Add(A=@n6, B=@n7), B=ScalarParameter() {parameterName=OffsetUVFlame}) {constB=0}) {materialFunction=&/MF_Flame}
  n9 = SmoothStep(Min=@n4.orangeColMin, Max=@n4.orangeColMin, Value=Multiply(A=@n5, B=@n8.Result)) {constMin=1}
  n10 = CurveAtlasRowParameter(CurveTime=@n9) {
   atlas=&VFX_CurveAtlas_Toon
   curve=&CC_EarthLane
   parameterName=Curve_Color
  }
  n11 = Constant3Vector() {constant=<rgba 0.4062 0.0539 0.0377 1>}
  n12 = Constant()
  n13 = Constant() {r=-0.3}
  n14 = Add(A=Add(A=@n6, B=AppendVector(A=@n12, B=@n13)), B=@n7)
  n15 = SmoothStep(
   Min=@n4.redCol
   Value=Multiply(
    A=@n5
    B=MaterialFunctionCall(UV_Mask=@n14) {materialFunction=&MF_Flame}.Result
   )
  ) {constMax=3.2, constMin=3.5}
  n16 = ScalarParameter() {defaultValue=0.3, parameterName=Emissive, sliderMax=2}
  n17 = Multiply(
   A=LinearInterpolate(A=@n10, B=LinearInterpolate(A=@n11, Alpha=@n15) {constB=0}, Alpha=OneMinus(None=@n15))
   B=Add(A=Multiply(A=@n9, B=@n16)) {constB=0.5}
  )
  Add(A=Multiply(A=Multiply(A=@n3, B=@n3), B=ScalarParameter() {defaultValue=0.05, parameterName=Vertical Gradient Bloom, sliderMax=1}), B=@n17)
  Add(
   A=TextureCoordinate()
   B=Multiply(
    A=NamedRerouteUsage()
    B=ScalarParameter() {defaultValue=0.05, group=UV Distortion, parameterName=Flame Distortion Strength, sliderMax=1, sliderMin=-1}
   ) {constB=0.1}
  )
  n18 = DynamicParameter() {parameterIndex=1, paramNames [X,Y,Z,FadeAmt]}
  n19 = MaterialFunctionCall(X=@n18.X, Y=@n18.X, Z=@n18.X) {materialFunction=&MakeFloat3}
  n20 = ParticlePositionWS()
  n21 = ScalarParameter() {defaultValue=20, parameterName=WPO Max Offset, sliderMax=50}
  n22 = Multiply(
   A=VectorParameter() {parameterName=WPO Axis Mask, defaultValue=<rgba 1 1 1 1>}.RGB
   B=Clamp(None=Subtract(A=@n19.Result, B=@n20), Min=Multiply(A=@n21) {constB=-1}, Max=@n21)
  )
  n23 = TextureCoordinate() {coordinateIndex=1}
  n24 = Power(Base=OneMinus(None=ComponentMask(None=@n23) {r=true}), Exp=ScalarParameter() {defaultValue=1, parameterName=WPO Bend, sliderMax=10})
  Add(A=Multiply(A=ScalarParameter() {defaultValue=1, parameterName=WPO Intensity, sliderMax=100}, B=LinearInterpolate(B=@n22, Alpha=@n24)))
  n25 = DotProduct(A=CameraVectorWS(), B=VertexNormalWS())
  Clamp(None=Power(Base=Abs(None=@n25), Exp=@n18.FadeAmt))
  Constant3Vector() {constant=<rgba 0.7182 0.4166 1 1>}
  Constant() {r=1}
  n26 = SphereMask(
   A=TextureCoordinate()
   B=AppendVector(A=ScalarParameter() {defaultValue=0.5, parameterName=PivSp}, B=ScalarParameter() {defaultValue=1, parameterName=PivSpY})
   Radius=DynamicParameter() {defaultValue=<rgba 3.5 1 1 0>, paramNames [redCol,orangeCol,orangeColMin,erode]}.erode
  ) {hardnessPercent=1}
  Divide(A=@n8.Result, B=@n26)
  DynamicParameter() {parameterIndex=2, paramNames [Param1,Param2,Param3,FadeAmt]}
  n27 = ParticleColor()
  n28 = ScalarParameter() {parameterName=useParticleColor}
  MaterialFunctionCall(Emissive=Multiply(A=@n17, B=LinearInterpolate(B=@n27.RGB, Alpha=@n28) {constA=1})) {materialFunction=&MF_FX_EmissiveAdjustments}
  Multiply(
   A=Subtract(A=Clamp(None=@n8.Result), B=@n26)
   B=MaterialFunctionCall("FadeSensitivity (optional)"=@n18.FadeAmt) {materialFunction=&FadeBasedOnViewAngle}.Result
  )
  Multiply(A=NamedRerouteUsage(), B=ScalarParameter() {defaultValue=50, parameterName=WPO_AMT})
  n29 = Constant3Vector() {constant=<rgba 0 1 0 0>}
  n30 = TextureCoordinate() {coordinateIndex=1}
  n31 = AppendVector(A=Constant(), B=Constant())
  n32 = OneMinus(None=ComponentMask(None=Add(A=@n30, B=@n31)) {r=true})
  n33 = Multiply(A=@n24, B=OneMinus(None=@n32))
  Multiply(A=@n29.B, B=@n33)
  Multiply(A=@n8.Result)
  n34 = SmoothStep(Value=@n8.Result) {constMax=0.3}
  n35 = DepthFade(Opacity=@n34, FadeDistance=ScalarParameter() {defaultValue=200, parameterName=Depth Fade Distance, sliderMax=100}) {fadeDistanceDefault=25}
  Multiply(
   A=ShadingPathSwitch(Default=@n35, Deferred=@n35, Forward=@n34, Mobile=@n34)
   B=ScalarParameter() {defaultValue=4, parameterName=Opacity, sliderMax=1}
  )
  n36 = VectorParameter() {parameterName=TOD color, defaultValue=<rgba 1 1 1 0.5>}
  Multiply(
   A=@n17
   B=MaterialFunctionCall(Value_Morning=@n36.R, Value_Day=@n36.R, Value_Evening=@n36.R, Value_Night=@n36.R) {materialFunction=&Scalar_Time_of_Day}.Result
  )
  n37 = ScalarParameter() {group=UV Distortion, parameterName=UV Speed X, sliderMax=5, sliderMin=-5}
  n38 = ScalarParameter() {defaultValue=2, group=UV Distortion, parameterName=UV Speed Y, sliderMax=5, sliderMin=-5}
  n39 = ScalarParameter() {defaultValue=1, group=UV Distortion, parameterName=UV Scale X, sliderMax=5}
  n40 = ScalarParameter() {defaultValue=0.5, group=UV Distortion, parameterName=UV Scale Y, sliderMax=5}
  n41 = ScalarParameter() {group=UV Distortion, parameterName=UV Speed X 2, sliderMax=5, sliderMin=-5}
  n42 = ScalarParameter() {group=UV Distortion, parameterName=UV Speed Y 2, sliderMax=5, sliderMin=-5}
  n43 = ScalarParameter() {defaultValue=1, group=UV Distortion, parameterName=UV Scale X 2, sliderMax=5}
  n44 = ScalarParameter() {defaultValue=0.25, group=UV Distortion, parameterName=UV Scale Y 2, sliderMax=5}
  NamedRerouteDeclaration(None=MaterialFunctionCall!(
    UV Speed=AppendVector(A=@n37, B=@n38)
    UV Scale=AppendVector(A=@n39, B=@n40)
    UV2 Speed=AppendVector(A=@n41, B=@n42)
    UV2 Scale=AppendVector(A=@n43, B=@n44)
    U Offset=ScalarParameter() {parameterName="U Offset (S)"}
   ) {materialFunction=/BRCosmetics/Accessories/FORT_Backpacks/Backpack_M_MED_EarthLane/FX/MF_FlameDistortion}.Result) {name=Distortion, nodeColor=<rgba 1 0.3281 0 1>}
  NamedRerouteDeclaration(None=LinearInterpolate(A=@n29, B=Constant3Vector(), Alpha=@n33)) {name=WPO, nodeColor=<rgba 0 0.3438 1 1>}
  NamedRerouteUsage()
  Power(Base=@n32, Exp=ScalarParameter() {defaultValue=1, parameterName=WPO Bend, sliderMax=10})
  ScalarParameter() {defaultValue=5, parameterName=RadiusSp}
  VertexColor()
 ]
}
