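# Translucent unlit material for the PillowMill glider headlight glow, enabled for Niagara sprite and mesh particles.
# A panning noise texture blends two pale yellows, scaled by a time-of-day brightness that rises at evening and night.
# A faint alpha combines vertex color, a time-of-day scalar and a powered SphereMask on inverted V coordinates, forming a soft falloff along the UVs.
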Material /BRCosmetics/Gadgets/Assets/VinderTech_GliderChute/Glider_PillowMill/Materials/M_PillowMill_Headlight
graph {
 properties {
  bAutomaticallySetUsageInEditor=false
  bUsedWithNiagaraMeshParticles=true
  bUsedWithNiagaraSprites=true
  blendMode=BLEND_Translucent
  shadingModel=MSM_Unlit
  usedShadingModels=Unlit
 }
 nodes [
  Multiply(A=VertexColor(), B=ParticleColor().RGB)
  n1 = Panner() {speedX=1}
  n2 = ScalarParameter() {defaultValue=2, parameterName=TOD_Morning}
  n3 = ScalarParameter() {defaultValue=8, parameterName=TOD_Night}
  Multiply(
   A=LinearInterpolate(
    A=Constant3Vector() {constant=<rgba 1 1 0.2969 1>}
    B=Constant3Vector() {constant=<rgba 0.9991 1 0.4479 1>}
    Alpha=TextureSample(UVs=@n1) {texture=/BRCosmetics/Gadgets/Assets/VinderTech_GliderChute/Glider_PillowMill/Textures/CINE_T_SimpleNoise}.R
   )
   B=MaterialFunctionCall(Value_Morning=@n2, Value_Day=@n2, Value_Evening=@n3, Value_Night=@n3) {materialFunction=&Scalar_Time_of_Day}.Result
  ) {constB=10}
  n4 = VertexColor()
  n5 = ScalarParameter() {defaultValue=0.06, parameterName=TOD_Alpha}
  n6 = ScalarParameter() {defaultValue=0.06, parameterName=TOD_AlphaNight}
  n7 = MaterialFunctionCall(Value_Morning=@n5, Value_Day=@n5, Value_Evening=@n6, Value_Night=@n6) {materialFunction=&Scalar_Time_of_Day}
  n8 = ComponentMask(None=TextureCoordinate()) {g=true}
  n9 = Constant()
  n10 = Constant()
  n11 = SphereMask(
   A=Add(A=OneMinus(None=@n8), B=AppendVector(A=@n9, B=@n10))
   B=ScalarParameter() {parameterName=B}
   Radius=ScalarParameter() {defaultValue=0.8, parameterName=Radius}
   Hardness=ScalarParameter() {defaultValue=0.001, parameterName=Hardness}
  )
  n12 = ScalarParameter() {defaultValue=3, parameterName=Exp}
  Multiply(A=Multiply(A=Multiply(A=@n4, B=@n7.Result), B=Power(Base=@n11, Exp=@n12)))
 ]
}
